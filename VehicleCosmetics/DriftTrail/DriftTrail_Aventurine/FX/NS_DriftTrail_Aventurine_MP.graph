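# Niagara system for the Aventurine drift trail vehicle cosmetic (MP version), drawing glowing, mostly cyan effects at the wheels.
# Ribbons with sine-animated position and facing and three renderers trail behind, alongside dot and spark sprites.
# A pulsing wheel-wrap mesh and orbiting triangle sprites follow wheel scale and take User painted colors.
# Emitters sample skeletal mesh bones or sockets, and FN_DisableTireFX masks particles per wheel via partition or spawn group.
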NiagaraSystem /VehicleCosmetics/DriftTrail/DriftTrail_Aventurine/FX/NS_DriftTrail_Aventurine_MP
dependencies {
 usedDataInterfaces [
  &NiagaraDataInterfaceCurve
  &NiagaraDataInterfaceArrayFloat3
  &NiagaraDataInterfaceSpriteRendererInfo
  &NiagaraDataInterfaceArrayInt32
  &NiagaraDataInterfaceSkeletalMesh
 ]
 usedDynamicInputs [&SelectIntFromArray]
 usedModules [
  &self:SystemSpawnScript.SystemScriptSource.SystemScriptGraph.NiagaraNodeAssignment_0.SetVariables_#1
  &FN_UnpackVehicleFlags
  &SystemState
  /Game/Effects/Niagara/Modules/Vehicle/FN_UpdateVehicleAttribs
  /Game/Effects/Niagara/Modules/Vehicle/FN_VehicleFlags_Cheatcodes
 ]
}
emitters {
 Dot_LocalSpace001 {
  emitterUpdateScript [
   &EmitterState()
   &SpawnBurst_Instantaneous(Spawn Count=2, Use Spawn Probability=false, UseLoopCountLimit=false) {enabled=false}
   &SpawnRate(SpawnRate=4, Use Spawn Probability=false)
  ]
  particleSpawnScript [
   &VC_PaintedVariantsOverride() {enabled=false}
   &InitializeParticle(
    Color=<rgba 0.25 1 1 1>
    Sprite Rotation Mode=::NewEnumerator2
    Sprite Size Mode=::NewEnumerator3
    Uniform Sprite Size=80
    UsePositionOffset=false
    Write Material Random=true
   )
   &PartitionParticles()
   &SkeletalMeshLocation(
    Bone / Socket Index=&Add_Integer(A=$Particles.PartitionParticles.Partition, B=4)
    Bone / Socket Sampling Mode=::NewEnumerator5
    Interpolate Mesh Sampling=false
    MaskSpawnGroup=false
    Mesh Sampling Type=::NewEnumerator6
    Sampled Position Offset=&AddVectorToPosition(Position=$User.offset, Vector=<-65 0 -15>)
    Skeletal Mesh=$System.SkeletalMesh
   )
   &AddVelocity(Velocity=<0 0 0>) {enabled=false}
  ]
  particleUpdateScript [
   &FN_DisableTireFX(
    Mask Index Attribute:NiagaraInt32=$Particles.PartitionParticles.Partition
    Mask Index FL=2
    Mask Index FR=3
    Mask Index RL=0
    Mask Index RR=1
   )
   &ParticleState()
   &ScaleColor(
    Scale Alpha=&FloatFromCurve(FloatCurve=DataInterface() {propertyValues {Curve {keys [
         {arriveTangent=8.0688, interpMode=RCIM_Cubic, leaveTangent=8.0688, tangentMode=RCTM_User, time=0, value=0}
         {arriveTangent=2, interpMode=RCIM_Cubic, leaveTangent=2, time=0.1, value=0.85}
         {arriveTangent=0.4232, interpMode=RCIM_Cubic, leaveTangent=0.4232, tangentMode=RCTM_User, time=0.5, value=1}
         {tangentMode=RCTM_User, time=1, value=0}
        ]}}})
    Scale RGB=<3 3 3>
    ScaleA=true
    ScaleRGB=true
   )
   &DynamicMaterialParameters(
    Index 0 Param 1=1
    Module.Index0Param0Write=true
    Module.Index0Param1Write=true
    Module.Index0Param2Write=true
    Module.Index0Param3Write=true
   )
   &CameraOffset(Camera Offset Amount=8)
   &SolveForcesAndVelocity(Clamp Velocity=false, Limit Acceleration=false)
  ]
  rendererClasses [&NiagaraSpriteRendererProperties]
  renderers [{rendererClass=&NiagaraSpriteRendererProperties}]
 }
 Ribbons001 {
  emitterSpawnScript [/SetVariables(Emitter.R1 Width=14) {isSetParameters=true}]
  emitterUpdateScript [
   &EmitterState()
   &VC_RateQualityHandler(High Spawn Rate=60, Low Spawn Rate=60, Medium Spawn Rate=60, Use Custom Rate?=true)
   &VC_SpawnRibbon_v1(Number of Ribbons=2, SpawnRate=$Emitter.Spawn Rate)
  ]
  particleSpawnScript [
   /SetVariables() {isSetParameters=true}
   /SetVariables(Particles.Position=&AddVectorToPosition(
     Position=$Particles.Position
     Vector=&TransformVector(
      Vector=&MakeVector(X=0, Y=&MaskFloatBySpawnGroup(Masked Float=10, Passthrough Float=-10), Z=0)
      Destination Space=::Simulation
     )
    )) {isSetParameters=true}
   &VC_PaintedVariantsOverride() {enabled=false}
   &VC_InitializeRibbon_v1(
    Color=&MakeCustomLinearColorFromBool(Bool=$User.bIsFrontendPreview, True Color=<rgba 0.1849 0.7396 0.7396 1>, False Color=<rgba 0.1849 0.7396 0.7396 1>)
    Custom Offset=true
    Enable Skeletal Position=true
    Lifetime=0.6
    Local Vector Offset=&AddVectorToPosition(Position=$User.offset, Vector=<-40 0 -15>)
    Offset Amount=1
    Ribbon Width=85
   )
   &AddVelocity(Velocity=<-1500 0 0>) {module=AddVelocity001}
   /VehicleCosmetics/FX/Niagara/Modules/MergeEmitters/VC_ExtraBindingColors(ExtraColors=1, R1 Color:LinearColor=<rgba 0 0.3385 0.2809 1>)
  ]
  particleUpdateScript [
   /SetVariables(Particles.RibbonFacing=&MaskVectorBySpawnGroup(
     Masked Vector=&MakeVector(
      X=0
      Y=0
      Z=&Sine(
       Normalized Angle=&Add_Float(
        A=&Multiply_Float(A=$Particles.NormalizedAge, B=1)
        B=&Multiply_Float(A=$Emitter.LoopedAge, B=0.1)
       )
       Period=1
       Scale=&FloatFromCurve(FloatCurve=DataInterface() {propertyValues {Curve {
           defaultValue=1
           keys [
            {arriveTangent=4.1785, interpMode=RCIM_Cubic, leaveTangent=4.1785, tangentMode=RCTM_User, time=0, value=0}
            {arriveTangent=0.0224, interpMode=RCIM_Cubic, leaveTangent=0.0224, tangentMode=RCTM_User, time=0.5, value=1}
           ]
          }}})
       Bias=1
      )
     )
     Passthrough Vector=&MakeVector(
      X=0
      Y=0
      Z=&Sine(
       Normalized Angle=&Add_Float(
        A=&Multiply_Float(A=$Particles.NormalizedAge, B=1)
        B=&Multiply_Float(A=$Emitter.LoopedAge, B=0.1)
       )
       Period=1
       Scale=&FloatFromCurve(FloatCurve=DataInterface() {propertyValues {Curve {
           defaultValue=1
           keys [
            {arriveTangent=4.1785, interpMode=RCIM_Cubic, leaveTangent=4.1785, tangentMode=RCTM_User, time=0, value=0}
            {arriveTangent=0.0224, interpMode=RCIM_Cubic, leaveTangent=0.0224, tangentMode=RCTM_User, time=0.5, value=1}
           ]
          }}})
       Bias=1
      )
     )
    )) {isSetParameters=true}
   /SetVariables(Particles.Position=&AddVectorToPosition(
     Position=$Particles.Position
     Vector=&TransformVector(
      Vector=&MakeVector(
       X=0
       Y=0
       Z=&Sine(
        Normalized Angle=&Add_Float(
         A=&Multiply_Float(A=$Particles.NormalizedAge, B=-1.3)
         B=&Multiply_Float(A=$Emitter.LoopedAge, B=1)
        )
        Period=0.5
        Scale=&FloatFromCurve(FloatCurve=DataInterface() {propertyValues {Curve {keys [<at 0.1 0>,<at 0.5 1>]}}})
        Bias=0
       )
      )
      Destination Space=::Simulation
      Apply Scale=true
     )
    )) {isSetParameters=true}
   &Color(Color=<rgba 0.25 1 1 1>) {enabled=false}
   &FN_DisableTireFX(Mask Index Attribute:NiagaraInt32=$Particles.SpawnGroup, Mask Index FL=2, Mask Index FR=3, Mask Index RL=0, Mask Index RR=1)
   &ParticleState()
   &InheritSourceMovement(
    Applied Position Scale=&Multiply_VectorByFloat(
     Vector=<1 1 1>
     Float=&FloatFromCurve(FloatCurve=DataInterface() {propertyValues {Curve {keys [
          {interpMode=RCIM_Cubic, leaveTangent=-0.1311, tangentMode=RCTM_Break, time=0, value=1}
          {arriveTangent=-0.0803, interpMode=RCIM_Cubic, tangentMode=RCTM_Break, time=1, value=0.9}
         ]}}})
    )
    ApplyForce=false
    ApplyPosition=true
    ApplyPositionFalloff=false
   )
   &SolveForcesAndVelocity(Clamp Velocity=false, Limit Acceleration=false)
   /SetVariables(
    Particles.R1 Position Offset:NiagaraPosition=&AddVectorToPosition(Position=$Particles.Position, Vector=/VehicleCosmetics/FX/Niagara/Modules/VC_ConvertVectorToWorld(Vector:Vector3f=<12 0 0>))
    Particles.R1 Scale Color:LinearColor=&VC_DynamicScaleColor(
     Color to Scale:LinearColor=$Particles.R1 Color
     RGB Scale:Vector3f=<0 1.8 1.8>
     Alpha Scale:NiagaraFloat=&FloatFromCurve(FloatCurve=DataInterface() {propertyValues {Curve {keys [<at 0 1>,<at 0.7 1>,<at 0.9 0>]}}})
    )
    Particles.R1 Width:NiagaraFloat=$Emitter.R1 Width
   ) {isSetParameters=true}
   /SetVariables(
    Particles.R2Width:NiagaraFloat=&SelectFloatFromArray(Direct Array Index=&SpawnGroup(), Float Selection Array=DataInterface() {propertyValues {FloatData [125,0]}})
    Particles.R3Width:NiagaraFloat=&SelectFloatFromArray(Direct Array Index=&SpawnGroup(), Float Selection Array=DataInterface() {propertyValues {FloatData [0,-125]}})
   ) {isSetParameters=true}
   &VC_UpdateRibbon_v1(Enable Skeletal Position=true)
   &ScaleRibbonWidth(Ribbon Width Scale=&FloatFromCurve(FloatCurve=DataInterface() {propertyValues {Curve {keys [<at 0 1.2>,<at 0.5 1>,<at 1 1>]}}}))
   &DynamicMaterialParameters(
    Index 0 Param 1=&MaskFloatBySpawnGroup(Masked Float=0, Passthrough Float=1)
    Index 1 Param 1=&MaskFloatBySpawnGroup(Masked Float=25, Passthrough Float=-25)
    Module.Index0Param0Write=true
    Module.Index0Param1Write=true
    Module.Index0Param2Write=true
    Module.Index0Param3Write=true
    Module.Index1Param0Write=true
    Module.Index1Param1Write=true
    Module.Index1Param2Write=true
    Module.Index1Param3Write=true
    Write Parameter Index 1=true
   )
   &ScaleColor(
    Scale Alpha=&FloatFromCurve(FloatCurve=DataInterface() {propertyValues {Curve {keys [<at 0 1>, {arriveTangent=-1.1765, leaveTangent=-1.8376, time=0.65, value=1}, {arriveTangent=-0.7092, tangentMode=RCTM_User, time=1, value=0}]}}})
    ScaleA=true
    ScaleRGB=true
   )
  ]
  rendererClasses [&NiagaraRibbonRendererProperties]
  renderers [
   {rendererClass=&NiagaraRibbonRendererProperties}
   {rendererClass=&NiagaraRibbonRendererProperties, rendererIndex=1}
   {rendererClass=&NiagaraRibbonRendererProperties, rendererIndex=2}
  ]
 }
 Sprites001 {
  emitterUpdateScript [
   &EmitterState()
   &SpawnRate(SpawnRate=25, Use Spawn Probability=false)
   &SpawnRate(SpawnGroup=1, SpawnRate=15, Use Spawn Probability=false) {module=SpawnRate001}
  ]
  particleSpawnScript [
   /SetVariables(Particles.VisTag:NiagaraInt32=&SpawnGroup()) {isSetParameters=true}
   &VC_PaintedVariantsOverride() {enabled=false}
   &InitializeParticle(
    Color=/VehicleCosmetics/FX/Niagara/Modules/MergeEmitters/VC_Dynamic_ColorBySpawnGroup(
     Color SG 0:LinearColor=<rgba 0.15 0.8 1 1>
     MaximumSpawnGroups:NiagaraInt32=2
     Color SG 1:LinearColor=<rgba 0.25 1 1 1>
     Color SG 2:LinearColor=<rgba 1 1 1 1>
     Color SG 3:LinearColor=<rgba 1 1 1 1>
     Color SG 4:LinearColor=<rgba 1 1 1 1>
    )
    Lifetime=&MaskFloatBySpawnGroup(
     Masked Float=&RandomRangeFloat(Minimum=0.35, Maximum=0.5)
     Passthrough Float=&RandomRangeFloat(Minimum=0.5, Maximum=0.65)
    )
    Sprite Rotation Angle=&MaskFloatBySpawnGroup(Masked Float=0, Passthrough Float=&RandomRangeFloat(Minimum=0.8, Maximum=1))
    Sprite Rotation Mode=::NewEnumerator1
    Sprite Size Mode=::NewEnumerator3
    Uniform Sprite Size=&MaskFloatBySpawnGroup(Masked Float=&RandomRangeFloat(Minimum=5, Maximum=10), Passthrough Float=30)
    UsePositionOffset=false
    Write Material Random=true
   )
   &PartitionParticles()
   &SkeletalMeshLocation(
    Bone / Socket Index=&Add_Integer(A=$Particles.PartitionParticles.Partition, B=4)
    Bone / Socket Sampling Mode=::NewEnumerator5
    Interpolate Mesh Sampling=false
    MaskSpawnGroup=false
    Mesh Sampling Type=::NewEnumerator6
    Sampled Position Offset=&AddVectorToPosition(
     Position=$User.offset
     Vector=&MaskVectorBySpawnGroup(Masked Vector=<-62 0 -25>, Passthrough Vector=<-58 0 -15>)
    )
    Skeletal Mesh=$System.SkeletalMesh
   )
   &ShapeLocation(
    MaskSpawnGroup=false
    Override Seed=false
    Sphere Radius=&MaskFloatBySpawnGroup(Masked Float=20, Passthrough Float=1)
    UseOwnerScale=true
   )
   &AddVelocity(
    Velocity=&MaskVectorBySpawnGroup(
     Masked Vector=&RandomRangeVector(Minimum=<-1000 0 -6>, Maximum=<-200 0 5>)
     Passthrough Vector=&RandomRangeVector(
      Minimum=&SelectVectorFromArray(
       Array Sampling Mode=::NewEnumerator0
       Direct Array Index=?
       Vector Selection Array=DataInterface() {propertyValues {FloatData [<-100 -20 35>, <-100 -20 0>]}}
       Random Seed=0
      )
      Maximum=<-200 20 40>
     )
    )
    Velocity Speed Scale=&MaskFloatBySpawnGroup(Masked Float=1, Passthrough Float=1.5)
   )
  ]
  particleUpdateScript [
   &FN_DisableTireFX(
    Mask Index Attribute:NiagaraInt32=$Particles.PartitionParticles.Partition
    Mask Index FL=2
    Mask Index FR=3
    Mask Index RL=0
    Mask Index RR=1
   )
   &ParticleState()
   &ScaleColor(
    Scale Alpha=&MaskFloatBySpawnGroup(Masked Float=1, Passthrough Float=3)
    Scale RGB=&MaskVectorBySpawnGroup(Masked Vector=<1 3 3>, Passthrough Vector=<1.5 7 7>)
    ScaleA=true
    ScaleRGB=true
   )
   &DynamicMaterialParameters(
    Index 0 Param 1=1
    Index 0 Param 3=2
    Module.Index0Param0Write=true
    Module.Index0Param1Write=true
    Module.Index0Param2Write=true
    Module.Index0Param3Write=true
   )
   &CameraOffset(Camera Offset Amount=&MaskFloatBySpawnGroup(Masked Float=0, Passthrough Float=30))
   &SubUVAnimation(
    Evaluation Type=::NewEnumerator0
    Output Alternate Renderer Binding=true
    "Sprite Renderer "=DataInterface() {propertyValues {SpriteRenderer=&self:Sprites001_0.NiagaraSpriteRendererProperties_0}}
    SubUV Animation Mode:ENiagara_SubUVLookupModeV2=::NewEnumerator2
    UseEndFrame=false
    UseStartFrame=false
   )
   &InheritSourceMovement(
    Applied Position Scale=&Multiply_VectorByFloat(
     Vector=<1 1 1>
     Float=&FloatFromCurve(FloatCurve=DataInterface() {propertyValues {Curve {keys [
          {interpMode=RCIM_Cubic, leaveTangent=-0.1105, tangentMode=RCTM_Break, time=0, value=1}
          {arriveTangent=-0.0803, interpMode=RCIM_Cubic, tangentMode=RCTM_Break, time=1, value=0.9}
         ]}}})
    )
    ApplyForce=false
    ApplyPosition=true
    ApplyPositionFalloff=false
   )
   &ScaleSpriteSize(Uniform Curve Sprite Scale=DataInterface() {propertyValues {Curve {keys [{tangentMode=RCTM_User, time=0, value=1}, <at 0.65 1>, {tangentMode=RCTM_User, time=1, value=0}]}}})
   &Drag(
    Drag=&MaskFloatBySpawnGroup(Masked Float=0, Passthrough Float=0.8)
    Rotational Drag:NiagaraFloat=&MaskFloatBySpawnGroup(Masked Float=0, Passthrough Float=1)
    Use Linear Drag=true
    Use Rotational Drag=true
   )
   &SolveForcesAndVelocity(Clamp Velocity=false, Limit Acceleration=false)
  ]
  rendererClasses [&NiagaraSpriteRendererProperties]
  renderers [
   {rendererClass=&NiagaraSpriteRendererProperties}
   {rendererClass=&NiagaraSpriteRendererProperties, rendererIndex=1}
  ]
 }
 WheelWrap001 {
  emitterUpdateScript [
   &EmitterState(
    Inactive Response=::NewEnumerator1
    Life Cycle Mode=::NewEnumerator1
    Loop Behavior=::NewEnumerator1
    Loop Duration Mode=::NewEnumerator1
    UseLoopDelay=false
   )
   &SpawnBurst_Instantaneous(Spawn Count=2, Use Spawn Probability=false, UseLoopCountLimit=false)
  ]
  particleSpawnScript [
   &VC_PaintedVariantsOverride() {enabled=false}
   &InitializeParticle(
    Color Mode=::NewEnumerator0
    Lifetime=5
    Mesh Scale Mode=::NewEnumerator4
    Sprite Rotation Mode=::NewEnumerator2
    UsePositionOffset=false
    Write Material Random=true
   )
   &PartitionParticles()
   &SkeletalMeshLocation(
    Bone / Socket Index=&Add_Integer(A=$Particles.PartitionParticles.Partition, B=4)
    Bone / Socket Sampling Mode=::NewEnumerator5
    Interpolate Mesh Sampling=false
    MaskSpawnGroup=false
    Mesh Sampling Type=::NewEnumerator6
    Skeletal Mesh=$System.SkeletalMesh
   )
   &Color(Color=&MakeCustomLinearColorFromBool(Bool=$User.bIsPainted, True Color=$User.PaintedColorMainTrail, False Color=<rgba 0.051 0.098 1 1>)) {module=Color001}
   &InitialMeshOrientation(Mesh Orientation Mode=::NewEnumerator7, Rotation=<0 0 0.5>, Use Rotation Vector=true)
   &Color(Color=<rgba 0 1 0 1>)
  ]
  particleUpdateScript [
   &FN_DisableTireFX(
    Mask Index Attribute:NiagaraInt32=$Particles.PartitionParticles.Partition
    Mask Index FL=2
    Mask Index FR=3
    Mask Index RL=0
    Mask Index RR=1
   )
   &VC_WheelWrapScale(Wheel Index Attribute:NiagaraInt32=&Add_Integer(A=$Particles.PartitionParticles.Partition, B=2))
   &DynamicMaterialParameters(
    Index 0 Param 1=3
    Index 0 Param 2=1
    Index 0 Param 3=2.5
    Index 0 Param 4=1
    Module.Index0Param0Write=true
    Module.Index0Param1Write=false
    Module.Index0Param2Write=false
    Module.Index0Param3Write=false
   )
   &ParticleState(Kill Particles When Lifetime Has Elapsed=false, Loop Particles Lifetime=true)
   &UpdateMeshOrientation(Rotation Coordinate Space=::NewEnumerator2, Rotation Vector=<0 1 0>) {enabled=false}
   &SkeletalMeshLocation(
    Bone / Socket Index=&Add_Integer(A=$Particles.PartitionParticles.Partition, B=4)
    Bone / Socket Sampling Mode=::NewEnumerator5
    MaskSpawnGroup=false
    Mesh Sampling Type=::NewEnumerator6
    Skeletal Mesh=$System.SkeletalMesh
   ) {module=SkeletalMeshLocation001}
   &SolveForcesAndVelocity(Clamp Velocity=false, Limit Acceleration=false)
   &ScaleColor(
    Scale RGB=&VectorFromFloat(Value=&Waveform("Waveform [1]"=::NewEnumerator0, "[1] Phase Scale"=5, "[1] Frequency"=0.5, "[1] Amplitude Min/Max"=<50 100>))
    ScaleA=true
    ScaleRGB=true
   )
  ]
  rendererClasses [&NiagaraMeshRendererProperties]
  renderers [{rendererClass=&NiagaraMeshRendererProperties, rendererIndex=0}]
 }
 Wheel_Triagles001 {
  emitterUpdateScript [
   &EmitterState()
   &SpawnRate(SpawnRate=40, Use Spawn Probability=false)
  ]
  particleSpawnScript [
   &VC_PaintedVariantsOverride() {enabled=false}
   &InitializeParticle(
    Lifetime Max=0.6
    Lifetime Min=0.7
    Lifetime Mode=::NewEnumerator1
    Position Mode=::NewEnumerator0
    Sprite Rotation Mode=::NewEnumerator2
    Sprite Size Mode=::NewEnumerator1
    Uniform Sprite Size Max=20
    Uniform Sprite Size Min=16
    Write Material Random=true
   )
   &PartitionParticles()
   &SkeletalMeshLocation(
    Bone / Socket Index=&Add_Integer(A=$Particles.PartitionParticles.Partition, B=0)
    Bone / Socket Sampling Mode=::NewEnumerator5
    Interpolate Mesh Sampling=false
    MaskSpawnGroup=false
    Mesh Sampling Type=::NewEnumerator6
    Skeletal Mesh=$System.SkeletalMesh
   )
   &AddVelocity() {enabled=false}
   &ShapeLocation(
    Handle Radius=2
    Large Radius=&RandomRangeFloat(Minimum=48, Maximum=50)
    MaskSpawnGroup=false
    Non Uniform Scale=<1 1 5>
    Offset=$Particles.SkeletalMeshLocation.SampledPosition
    Offset Mode=::NewEnumerator0
    Override Seed=false
    Rotation Angle:NiagaraFloat=&Multiply_Float(A=$Emitter.Age, B=2)
    Rotation Axis=<0 1 0>
    Rotation Mode=::NewEnumerator1
    Shape Origin=$Engine.Emitter.SimulationPosition
    Shape Primitive=::NewEnumerator3
    Surface Distribution=1
    UseOwnerScale=true
   ) {enabled=false}
   &Color(Color=&MakeCustomLinearColorFromBool(Bool=$User.bIsPainted, True Color=$User.PaintedColorTriangles, False Color=<rgba 1 1 1 1>))
  ]
  particleUpdateScript [
   &FN_DisableTireFX(
    Mask Index Attribute:NiagaraInt32=$Particles.PartitionParticles.Partition
    Mask Index FL=2
    Mask Index FR=3
    Mask Index RL=0
    Mask Index RR=1
   )
   &VC_WheelWrapScale(Wheel Index Attribute:NiagaraInt32=&Add_Integer(A=$Particles.PartitionParticles.Partition, B=2))
   &ScaleSpriteSize(Uniform Curve Sprite Scale=DataInterface() {propertyValues {Curve {keys [{tangentMode=RCTM_User, time=0, value=1.5}, {tangentMode=RCTM_User, time=1, value=0}]}}}) {enabled=false}
   &ShapeLocation(
    Handle Radius=2
    Large Radius=&Multiply_Float(
     A=&RandomRangeFloat(Minimum=50, Maximum=53)
     B=&MakeFloatFromVector(Vector=$Particles.Scale, Channel=::NewEnumerator0)
    )
    MaskSpawnGroup=false
    Non Uniform Scale=<1 1 5>
    Offset=$Particles.SkeletalMeshLocation.SampledPosition
    Offset Mode=::NewEnumerator0
    Override Seed=false
    Rotation Angle:NiagaraFloat=&Multiply_Float(A=$Emitter.Age, B=1.15)
    Rotation Axis=<0 2 0>
    Rotation Mode=::NewEnumerator1
    Shape Primitive=::NewEnumerator3
    Surface Distribution=1
    UseOwnerScale=true
   ) {module=ShapeLocation003}
   &ParticleState()
   &ScaleColor(
    Scale Alpha=&FloatFromCurve(FloatCurve=DataInterface() {propertyValues {Curve {keys [
         {arriveTangent=9.0725, interpMode=RCIM_Cubic, leaveTangent=9.0725, tangentMode=RCTM_User, time=0, value=0}
         {arriveTangent=-2.5249, interpMode=RCIM_Cubic, leaveTangent=-2.5249, tangentMode=RCTM_User, time=0.1408, value=0.9847}
         {arriveTangent=-0.0405, interpMode=RCIM_Cubic, leaveTangent=-0.0405, tangentMode=RCTM_User, time=1, value=0}
        ]}}})
    Scale RGB=<10 10 10>
    ScaleA=false
    ScaleRGB=true
   )
   &SubUVAnimation(
    Direct Frame Index:NiagaraFloat=&SelectFloatFromArray(
     Array Sampling Mode=::NewEnumerator0
     Direct Array Index=?
     Float Selection Array=DataInterface() {propertyValues {FloatData [0,2,3]}}
     Random Seed=0
    )
    "Sprite Renderer "=DataInterface() {propertyValues {SpriteRenderer=&self:Wheel_Triagles001_0.NiagaraSpriteRendererProperties_4}}
    SubUV Animation Mode:ENiagara_SubUVLookupModeV2=::NewEnumerator5
   )
   &DynamicMaterialParameters(
    Index 0 Param 1=1
    Index 0 Param 2=1
    Index 0 Param 3=&MakeCustomFloatFromBool(Bool=$User.bIsPainted, True Float=1, False Float=0)
    Module.Index0Param0Write=true
    Module.Index0Param1Write=true
    Module.Index0Param2Write=true
    Module.Index0Param3Write=false
   )
   &SolveForcesAndVelocity(Clamp Velocity=false, Limit Acceleration=false)
  ]
  rendererClasses [&NiagaraSpriteRendererProperties]
  renderers [{rendererClass=&NiagaraSpriteRendererProperties}]
 }
}
summary {
 systemName=NS_DriftTrail_Aventurine_MP
 emitters [
  {emitterName=Dot_LocalSpace001, rendererClasses [&NiagaraSpriteRendererProperties]}
  {emitterName=Ribbons001, rendererClasses [&NiagaraRibbonRendererProperties]}
  {emitterName=Sprites001, rendererClasses [&NiagaraSpriteRendererProperties]}
  {emitterName=Wheel_Triagles001, rendererClasses [&NiagaraSpriteRendererProperties]}
  {emitterName=WheelWrap001, rendererClasses [&NiagaraMeshRendererProperties]}
 ]
 userVariables [
  User.bIsFrontendPreview {type=NiagaraBool}
  User.bIsPainted {type=NiagaraBool}
  User.Brightness2 {type=NiagaraFloat, defaultValue=1, description=""}
  User.ColorVariantID {type=NiagaraInt32}
  User.offset {type=NiagaraPosition, description="", defaultValue {}}
  User.PaintedColor {type=LinearColor, description="", defaultValue {struct=LinearColor, value=<rgba 1 1 1 1>}}
  User.PaintedColorMainTrail {type=LinearColor, description="", defaultValue {struct=LinearColor, value=<rgba 1 1 1 1>}}
  User.PaintedColorTriangles {type=LinearColor, description="", defaultValue {struct=LinearColor, value=<rgba 1 1 1 1>}}
  User.PaintedScalar {type=NiagaraFloat, defaultValue=1, description=""}
  User.VehicleFlags {type=NiagaraInt32, defaultValue=806289408}
  User.Wheel_Radius_Array {
   type=NiagaraDataInterfaceArrayFloat
   defaultValue {
    struct=NiagaraExt_VariableValue_DataInterface
    value {
     dataInterface=&self:NiagaraDataInterfaceArrayFloat_0
     dataInterfaceClass=&NiagaraDataInterfaceArrayFloat
    }
   }
  }
  User.Wheel_Width_Array {
   type=NiagaraDataInterfaceArrayFloat
   defaultValue {
    struct=NiagaraExt_VariableValue_DataInterface
    value {
     dataInterface=&self:NiagaraDataInterfaceArrayFloat_1
     dataInterfaceClass=&NiagaraDataInterfaceArrayFloat
    }
   }
  }
  User.WheelSize {type=NiagaraFloat, defaultValue=0.1, description=""}
 ]
}
