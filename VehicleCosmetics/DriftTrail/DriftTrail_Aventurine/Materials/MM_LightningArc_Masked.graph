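# Masked unlit Niagara ribbon and mesh particle material for the Aventurine drift trail, drawing glowing lines and musical notes.
# Channels of one RGB texture are tinted separately for emissive, with lerps and an IsWheelWrap parameter for a wheel-wrap variant.
# The mask combines texture channels, a procedural circle, a U-edge ramp and particle alpha.
# Time-driven sine/cosine world position offset wobbles the trail, weighted along U and scaled by a dynamic parameter.
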Material /VehicleCosmetics/DriftTrail/DriftTrail_Aventurine/Materials/MM_LightningArc_Masked
graph {
 properties {
  bAutomaticallySetUsageInEditor=false
  bUsedWithNiagaraMeshParticles=true
  bUsedWithNiagaraRibbons=true
  blendMode=BLEND_Masked
  shadingModel=MSM_Unlit
  usedShadingModels=Unlit
 }
 nodes [
  n1 = VectorParameter() {parameterName=Notes_Color, defaultValue=<rgba 0.5857 1 1 1>}
  n2 = TextureSample(UVs=LinearInterpolate(A=TextureCoordinate() {uTiling=5.5}, B=TextureCoordinate(), Alpha=NamedRerouteUsage())) {texture=/VehicleCosmetics/DriftTrail/DriftTrail_Aventurine/Textures/T_VCVFX_Aventurine_RGB}
  n3 = LinearInterpolate(
   A=Multiply(A=Multiply(A=@n1.RGB, B=@n2.G), B=ScalarParameter() {defaultValue=5, group=General Parameters, parameterName=Musical Notes Brightness})
   Alpha=NamedRerouteUsage()
  ) {constB=0}
  n4 = NamedRerouteUsage()
  n5 = NamedRerouteDeclaration(None=@n2.B) {name=Mask Lines, nodeColor=<rgba 1 0 0.6562 1>}
  n6 = Multiply(
   A=Multiply(A=LinearInterpolate(A=@n2.R, B=@n5, Alpha=@n4), B=VectorParameter() {parameterName=Lines_Color, defaultValue=<rgba 0.2376 1 1 1>}.RGB)
   B=ScalarParameter() {defaultValue=6, group=General Parameters, parameterName=Brightness}
  )
  MaterialFunctionCall(Emissive=Multiply(A=Add(A=@n3, B=@n6), B=ScalarParameter() {defaultValue=1, group=General Parameters, parameterName=Brightness 2})) {materialFunction=&MF_FX_EmissiveAdjustments}
  MaterialExpressionVertexColor_2 = ParticleColor()
  n7 = Add(A=TextureCoordinate() {uTiling=0.65, vTiling=0.75}, B=Constant2Vector() {g=-0.385, r=-0.46})
  n8 = OneMinus(None=DotProduct(A=@n7, B=@n7))
  n9 = Constant() {r=0.8}
  MaterialExpressionConstantClamp_3 = Saturate(None=Multiply(A=Subtract(A=@n8, B=@n9), B=Constant() {r=100}))
  n10 = Multiply(A=NamedRerouteUsage(), B=@MaterialExpressionConstantClamp_3)
  n11 = TextureCoordinate()
  MaterialExpressionConstantClamp_2 = Saturate(None=ConstantBiasScale(None=ComponentMask(None=@n11) {r=true}) {bias=-0.005, scale=10})
  Multiply(
   A=@MaterialExpressionVertexColor_2.A
   B=LinearInterpolate(A=Multiply(A=@n10, B=@MaterialExpressionConstantClamp_2), B=NamedRerouteUsage(), Alpha=NamedRerouteUsage())
  )
  NamedRerouteDeclaration(None=MaterialFunctionCall(Speed=ScalarParameter() {defaultValue=1, group=WPO, parameterName=Speed}) {materialFunction=&TimeWithSpeedVariable}.Time) {name=Time WPO, nodeColor=<rgba 1 0 0.0799 1>}
  n12 = ComponentMask(None=TextureCoordinate()) {r=true}
  n13 = ScalarParameter() {defaultValue=0.8, parameterName=Initial_Fade}
  n14 = Divide(A=ComponentMask(None=TextureCoordinate()) {r=true}, B=ScalarParameter() {defaultValue=1, group=WPO, parameterName=tile})
  n15 = Cosine(None=Add(A=NamedRerouteUsage(), B=@n14))
  n16 = NamedRerouteUsage()
  n17 = MaterialFunctionCall(X=@n15, Y=Sine(None=Add(A=@n14, B=@n16)), Z=@n15) {materialFunction=&MakeFloat3}
  n18 = DynamicParameter() {paramNames [Intensity WPO, Musical Notes Brightness, Brightness, Music_Bri]}
  n19 = Transform(None=Multiply(A=Multiply(A=@n17.Result, B=@n18.Intensity WPO), B=ScalarParameter() {defaultValue=2, group=WPO, parameterName=Amount})) {transformSourceType=TRANSFORMSOURCE_Local}
  NamedRerouteDeclaration(None=Multiply(A=Power(Base=@n12, Exp=@n13), B=ConstantBiasScale(None=@n19))) {name=WPO, nodeColor=<rgba 0.0395 0.0561 0.1356 1>}
  NamedRerouteDeclaration(None=Max(A=@n2.R, B=@n2.G)) {name=Mask Lines and notes, nodeColor=<rgba 0 0.0156 1 1>}
  n20 = ScalarParameter() {group=General Parameters, parameterName="IsWheelWrap? 1=True 0=false"}
  NamedRerouteDeclaration(None=Saturate(None=Round(None=@n20) {desc=round the number to nearest integer}) {desc=Then saturate between 0 and 1}) {name=Is Wheel Wrap?, nodeColor=<rgba 1 0.2011 0 1>}
  Transform(None=NamedRerouteUsage())
 ]
}
